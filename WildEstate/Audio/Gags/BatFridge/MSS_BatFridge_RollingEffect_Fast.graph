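# Stereo MetaSound source for the BatFridge gag's fast rolling effect, layering a rolling loop wave with intermittent groans.
# The rolling wave starts on play, while a delayed trigger picks one random period that drives a repeat for the groans.
# Each repeat plays a random groan with slight pitch variation, mixed at reduced gain.
# OnFinished comes from a two-input accumulator with only the groan player's On Play wired.
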MetaSoundSource /WildEstate/Audio/Gags/BatFridge/MSS_BatFridge_RollingEffect_Fast
inputs [
 Sounds {type=WaveAsset:Array}
 UE.Source.OnPlay {type=Trigger}
 Wave Asset {type=WaveAsset}
]
interfaces [UE.OutputFormat.Stereo:1.0, UE.Source.OneShot:1.0, UE.Source:1.1]
outputs [
 "UE.OutputFormat.Stereo.Audio:0" {type=Audio}
 "UE.OutputFormat.Stereo.Audio:1" {type=Audio}
 UE.Source.OneShot.OnFinished {type=Trigger}
]
pages [Default {
  input_defaults {
   UE.Source.OnPlay=null
   Wave Asset=&/Phys_BatFridge_RollingEffects_Fast_LP
   Sounds [
    &/Phys_BatFridge_RollingEffects_Groan_01
    &/Phys_BatFridge_RollingEffects_Groan_02
    &/Phys_BatFridge_RollingEffects_Groan_03
    &/Phys_BatFridge_RollingEffects_Groan_04
   ]
  }
  nodes [
   TriggerDelay_0 = UE.Trigger Delay(In=^UE.Source.OnPlay)
   n1 = UE.RandomTime(Next=@TriggerDelay_0.Out) {also {defaults {Max=15, Min=8}}}
   ArrayRandomGet_0 = Array.Random Get:WaveAsset:Array(In Array=^Sounds, Next=UE.TriggerRepeat(Start=@n1.On Next, Period=@n1.Value).RepeatOut) {also {defaults {Weights []}}}
   n2 = UE.RandomFloat(Next=@ArrayRandomGet_0.On Next) {also {defaults {Max=1.02, Min=0.98}}}
   WavePlayer_0 = UE.Wave Player:Stereo(Wave Asset=@ArrayRandomGet_0.Value, Play=@n2.On Next, Pitch Shift=@n2.Value) {also {version [1,1]}}
   WavePlayer_1 = UE.Wave Player:Stereo(Play=^UE.Source.OnPlay, Wave Asset=^Wave Asset) {also {defaults {Loop="<MetaSoundBuilderResult.SUCCEEDED: 0>"}, version [1,1]}}
   AudioMixerAudioMixerStereo2_0 = "AudioMixer.Audio Mixer (Stereo, 2)"(In 0 L=@WavePlayer_0.Out Left, In 0 R=@WavePlayer_0.Out Right, In 1 L=@WavePlayer_1.Out Left, In 1 R=@WavePlayer_1.Out Right) {also {defaults {Gain 0=0.5}}}
   TriggerAccumulatorTriggerAccumulate2_0 = "TriggerAccumulator.Trigger Accumulate (2)"(In 0=@WavePlayer_0.On Play)
  ]
  links [
   {dst="@UE.OutputFormat.Stereo.Audio:0", src=AudioMixerAudioMixerStereo2_0.Out L}
   {dst="@UE.OutputFormat.Stereo.Audio:1", src=AudioMixerAudioMixerStereo2_0.Out R}
   {dst="@UE.Source.OneShot.OnFinished", src=TriggerAccumulatorTriggerAccumulate2_0.Out}
  ]
 }]
properties {
 assetImportData=&self:AssetImportData
 numChannels=2
 outputFormat=Stereo
 sampleRate=48000
}
